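# Physics asset for the Bright Bomber Mint backpack cosmetic, defining capsule bodies and joints for its dangling zipper and handles.
# The pack body combines four capsules and anchors the two handles, the zipper chain and the spring.
# The four zipper bodies link in series with narrow twist limits, while each handle joins the pack directly with wider twist limits.
# The spring body has all swing and twist locked to the pack.
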PhysicsAsset /BRCosmetics/Accessories/FORT_Backpacks/F_MED_Bright_Bomber_Mint/Meshes/F_MED_Bright_Bomber_Mint_Pack_Physics
bodies {
 dyn_handleA_1 [{shapeType=Capsule, length=9.4801, radius=0.6877, shapeName=dyn_handleB_1_capsule, center=<4 0 0>, rotation=<rot 0 0.0904 0>}]
 dyn_handleB_1 [{shapeType=Capsule, length=6.859, radius=1.2331, shapeName=dyn_handleB_1_capsule, center=<4 0 0>}]
 dyn_zipper_1 [{shapeType=Capsule, length=3.0016, radius=0.5, shapeName=dyn_zipper_1_capsule, center=<2.655 0 0>, rotation=<rot 90 0 0>}]
 dyn_zipper_2 [{shapeType=Capsule, length=0.311, radius=1, shapeName=dyn_zipper_2_capsule, center=<1.2813 0 0>, rotation=<rot 90 -0.0197 -0.0195>}]
 dyn_zipper_3 [{shapeType=Capsule, length=2.3296, radius=1, shapeName=dyn_zipper_3_capsule, center=<2.5138 0 0>, rotation=<rot 90 -0.0002 0.0002>}]
 dyn_zipper_4 [{shapeType=Capsule, length=1.9385, radius=1, shapeName=dyn_zipper_4_capsule, center=<5.3535 0 -0.0002>, rotation=<rot 90 0.0002 0.0005>}]
 pack [
  {
   shapeType=Capsule
   length=8.7942
   radius=7.9198
   shapeName=pack_capsule
   center=<3.7171 -3.0704 -0.6734>
   rotation=<rot 0.4914 109.2646 -6.1725>
  }
  {
   shapeType=Capsule
   length=0
   radius=8.362
   shapeName=pack_capsule
   center=<14.5069 0.7719 -4.5567>
   rotation=<rot 87.0554 -179.9994 162.4735>
  }
  {shapeType=Capsule, length=15, radius=15, shapeName=pack_capsule, center=<23.1612 21.846 -13.5861>, rotation=<rot 0 0 12.0808>}
  {shapeType=Capsule, length=15, radius=15, shapeName=pack_capsule, center=<33.8295 13.7559 4.2288>}
 ]
 spring [{shapeType=Capsule, length=3.5478, radius=1.9618, shapeName=spring_capsule, center=<3.1686 0.8776 0>}]
}
constraints [
 {bone1Name=dyn_zipper_2, bone2Name=dyn_zipper_1, swing1LimitDegrees=60, swing2LimitDegrees=20, twistLimitDegrees=2}
 {bone1Name=dyn_zipper_3, bone2Name=dyn_zipper_2, swing1LimitDegrees=90, swing2LimitDegrees=20, twistLimitDegrees=2}
 {bone1Name=dyn_zipper_4, bone2Name=dyn_zipper_3, swing1LimitDegrees=90, swing2LimitDegrees=20, twistLimitDegrees=2}
 {bone1Name=dyn_zipper_1, bone2Name=pack, swing1LimitDegrees=60, swing2LimitDegrees=45, twistLimitDegrees=2}
 {bone1Name=dyn_handleA_1, bone2Name=pack, swing1LimitDegrees=80, swing2LimitDegrees=30, twistLimitDegrees=20}
 {bone1Name=dyn_handleB_1, bone2Name=pack, swing1LimitDegrees=80, swing2LimitDegrees=30, twistLimitDegrees=20}
 {
  bone1Name=spring
  bone2Name=pack
  swing1LimitDegrees=45
  swing1Motion=Locked
  swing2LimitDegrees=45
  swing2Motion=Locked
  twistLimitDegrees=45
  twistMotion=Locked
 }
]
