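# Animation Blueprint for the mint Bright Bomber backpack, adding secondary physics motion that follows the wearer.
# Each update measures parent-mesh spine twist to weight an additive offset on the pack bone over the reference pose.
# That pose feeds a RigidBody node using the pack's physics asset and a gravity override from a deprecated state function with skydive wind parameters.
# An Anim Dynamics node for a zipper bone is left unconnected.
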AnimBlueprint /BRCosmetics/Accessories/FORT_Backpacks/F_MED_Bright_Bomber_Mint/Meshes/F_MED_Bright_Bomber_Mint_Pack_AnimBP
target_skeleton=&/F_MED_Bright_Bomber_Mint_Pack_Skeleton
graphs [
 AnimGraph {nodes [
   Anim Dynamics - Bone: dyn_zipper_4(Alpha=1, ExternalForce="0,0,0") {
    angularDampingOverride=1
    angularSpringConstant=20000
    bAngularSpring=true
    bGravityOverrideInSimSpace=true
    bOverrideAngularDamping=true
    bUseGravityOverride=true
    gravityScale=0.35
    simulationSpace=RootRelative
    alphaBoolBlend {blendOption=Linear}
    boundBone {boneName=dyn_zipper_4}
    componentLinearAccScale=<1 1 1>
    componentLinearVelScale=<1 1 1>
    gravityOverride {z=-980}
    physicsBodyDefinitions [{
      sphereCollisionRadius=0
      boundBone {boneName=dyn_zipper_4}
      boxExtents=<10 10 10>
      constraintSetup {angularLimitsMax=<15 15 15>, angularLimitsMin=<-15 -15 -15>, angularTarget=<1 0 0>}
      localJointOffset=<0 0 0>
     }]
   }
   n1 = "Transform (Modify) Bone - Bone: pack"(
    ComponentPose=Local To Component(LocalPose=Local Space Ref Pose())
    Alpha=Get spine_transform_delta_weight() -> (spine_transform_delta_weight:double)
    Translation="1,-1.5,0"
    Rotation="0,15,0"
   ) {
    rotationMode=BMM_Additive
    rotationSpace=BCS_ParentBoneSpace
    translationMode=BMM_Additive
    translationSpace=BCS_ParentBoneSpace
    alphaBoolBlend {blendOption=Linear}
    boneToModify {boneName=pack}
    rotation {yaw=15}
    translation=<1 -1.5>
   }
   n2 = RigidBody!(ComponentPose=@n1, Alpha, OverrideWorldGravity:Vector=Get Gravity Override(), ExternalForce) -> (Pose) {
    bOverrideWorldGravity=true
    overridePhysicsAsset=&/F_MED_Bright_Bomber_Mint_Pack_Physics
    simulationSpace=BaseBoneSpace
    alphaBoolBlend {blendOption=Linear}
    baseBoneRef {boneName=root}
    componentAppliedLinearAccClamp=<3000 3000 3000>
    componentLinearAccScale=<0.1 0.2 0.2>
    componentLinearVelScale=<0.1 0.2 0.2>
    overrideWorldGravity {z=-980}
   }
   Output Pose(Result=Component To Local(ComponentPose=@n2)) {layerGroup=None}
  ]}
 EventGraph {nodes [
   Event BlueprintUpdateAnimation()
   > Set spine_transform_delta_weight(spine_transform_delta_weight:double=GetAlphaFromReferenceTransformDelta(
     SkeletalMeshComponent=Get parent_skeletal_mesh() -> (parent_skeletal_mesh:Skeletal Mesh Component Object Reference)
     BoneName=spine_05
     BaseName=spine_01
     DeltaComponentType=QuaternionTwist
     InputRange=(X=0,Y=45)
     TwistAxis="0,0,1"
    )) -> (Output_Get:double)
   > n1 = Gravity Override State Function DEPRECATED(CharacterPartAnimInstance=Self-Reference(), GameTimeInSeconds=GetGameTimeInSeconds(), GravityOverrideParams=Get Gravity Override Params())
   > Set Gravity Override(Gravity Override=@n1)
   Event BlueprintInitializeAnimation()
   > Set parent_skeletal_mesh(parent_skeletal_mesh:Skeletal Mesh Component Object Reference=GetOwnerSkeletalMeshComponent()) -> (Output_Get:Skeletal Mesh Component Object Reference)
   TryGetPawnOwner()
  ]}
]
variables [
 Gravity Override {type=Vector}
 Gravity Override Params {
  type=GravityOverrideParamsStruct
  default {
   skydive_DivingDown {windAmplitude=<50 150 50>, windFrequency=<1.5 6 5>}
   skydive_FloatingDown {gravityOverride=<-50 -150 -150>, windAmplitude=<0 -35 35>, windOffset=<0 -350 300>}
   skydive_GoingUp {gravityOverride=<0 0 -100>, windOffset=<0 -600 400>}
  }
 }
 parent_skeletal_mesh {type=SkeletalMeshComponent}
 spine_transform_delta_weight {type=number}
]
